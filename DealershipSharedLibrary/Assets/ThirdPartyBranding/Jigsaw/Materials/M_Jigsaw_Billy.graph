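# Unlit translucent Niagara sprite material that shows the Jigsaw Billy hero image with glowing eyes, red by default.
# Eye-texture blue channel, boosted and tinted, adds to the brightened image, then blends with desaturated particle colour into an emissive adjustment function.
# Opacity combines texture and particle alpha, sine-based UV edge masks and depth fade.
# The time-driven eye pulse lerps identical inputs, so has no effect; an older colour/depth-fade branch dangles.
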Material /DealershipSharedLibrary/Assets/ThirdPartyBranding/Jigsaw/Materials/M_Jigsaw_Billy
graph {
 properties {bUsedWithNiagaraSprites=true, blendMode=BLEND_Translucent, shadingModel=MSM_Unlit, usedShadingModels=Unlit}
 nodes [
  MaterialExpressionVertexColor_2 = ParticleColor()
  n1 = TextureSample() {texture=/DealershipSharedLibrary/Assets/ThirdPartyBranding/Jigsaw/Textures/T_Jigsaw_BillyHero_D}
  n2 = Frac(None=Multiply(A=TextureCoordinate() {uTiling=0.25, vTiling=0.25}, B=Constant2Vector() {g=2, r=2}))
  n3 = ComponentMask(None=@n2) {r=true}
  MaterialExpressionConstantClamp_8 = Saturate(None=ConstantBiasScale(None=Sine(None=@n3)) {bias=-0.1, scale=1.8})
  n4 = ComponentMask(None=@n2) {g=true}
  MaterialExpressionConstantClamp_9 = Saturate(None=ConstantBiasScale(None=Sine(None=@n4)) {bias=-0.0025, scale=5})
  n5 = Multiply(A=@n1.A, B=Multiply(A=@MaterialExpressionConstantClamp_8, B=@MaterialExpressionConstantClamp_9))
  MaterialExpressionConstantClamp_6 = Saturate(None=MaterialFunctionCall(Opacity In=Multiply(A=@MaterialExpressionVertexColor_2.A, B=@n5), DepthFade Distance=Constant() {r=50}) {materialFunction=&MF_VCVFX_DepthFade}.Result)
  n6 = ParticleColor()
  n7 = TextureSample() {texture=&T_Jigsaw_BillyHero_D}
  DepthFade(Opacity=Multiply(A=@n6.A, B=@n7.A)) {fadeDistanceDefault=200}
  n8 = Constant() {r=0.65}
  LinearInterpolate(A=@n6.RGB, B=Multiply(A=Desaturation(None=@n6.RGB, Fraction=@n8), B=Multiply(A=@n7.RGB) {constB=2}), Alpha=@n7.A)
  n9 = Desaturation(None=@MaterialExpressionVertexColor_2.RGB, Fraction=Constant() {r=0.65})
  n10 = TextureSample() {texture=/DealershipSharedLibrary/Assets/ThirdPartyBranding/Jigsaw/Textures/T_Jigsaw_Eyes_D}
  n11 = TextureCoordinate() {vTiling=0}
  n12 = Time()
  n13 = Add(A=ComponentMask(None=Frac(None=@n11)) {g=true}, B=Multiply(A=OneMinus(None=@n12), B=Constant() {r=3}))
  MaterialExpressionConstantClamp_7 = Saturate(None=ConstantBiasScale(None=Sine(None=@n13)) {bias=-0.1, scale=1})
  n14 = LinearInterpolate(A=@n10.B, B=@n10.B, Alpha=@MaterialExpressionConstantClamp_7)
  n15 = LinearInterpolate(A=Constant() {r=60}, B=Constant() {r=100})
  n16 = Add(A=Multiply(A=@n1.RGB, B=Constant() {r=6}), B=Multiply(A=Multiply(A=@n14, B=@n15), B=VectorParameter() {defaultValue=<rgba 1 0 0 1>}.RGB))
  MaterialFunctionCall(Emissive=LinearInterpolate(A=@MaterialExpressionVertexColor_2.RGB, B=Multiply(A=@n9, B=@n16), Alpha=@n1.A)) {materialFunction=&MF_FX_EmissiveAdjustments}
 ]
}
