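# Unlit translucent particle material that draws a desaturated, inverted, posterized copy of the scene behind it with an animated TV-style interference pattern.
# Inverted scene color is raised to ColorPower and snapped to three levels with nested Ifs.
# Scrolling Fmod stripes on bump-offset, radially distorted screen UVs modulate it, scaled by TVEffectBrightness and particle color into MF_FX_EmissiveAdjustments.
# A circular SphereMask, particle alpha and MF_VCVFX_DepthFade form a soft-edged alpha term.
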Material /DealershipSharedLibrary/Assets/ThirdPartyBranding/Jigsaw/Materials/M_Jigsaw_SceneInvert
graph {
 properties {
  bUsedWithNanite=true
  bUsedWithNiagaraMeshParticles=true
  bUsedWithNiagaraSprites=true
  blendMode=BLEND_Translucent
  shadingModel=MSM_Unlit
  usedShadingModels=Unlit
 }
 nodes [
  n1 = DynamicParameter() {paramNames [FlowDistance, FallOff, ColorPower, TVEffectBrightness]}
  n2 = Constant() {r=9216}
  n3 = ScreenPosition()
  n4 = Add(A=Multiply(A=TextureCoordinate(), B=Constant() {r=2}), B=Constant() {r=-1})
  MaterialExpressionConstantClamp_3 = Saturate(None=DotProduct(A=@n4, B=@n4))
  n5 = OneMinus(None=@MaterialExpressionConstantClamp_3)
  MaterialExpressionConstantClamp_4 = Saturate(None=@MaterialExpressionConstantClamp_3)
  n6 = AppendVector(A=@n5, B=@MaterialExpressionConstantClamp_4)
  n7 = Add(A=ComponentMask(None=Multiply(A=@n2, B=@n3.ViewportUV)) {g=true, r=true}, B=@n6)
  n8 = BumpOffset(Coordinate=@n7, Height=Constant() {r=8})
  n9 = Multiply(A=Time(), B=Constant() {r=6})
  n10 = Fmod(A=Add(A=ComponentMask(None=@n8) {g=true}, B=@n9), B=Constant() {r=0.4})
  n11 = Multiply(A=Constant() {r=5120}, B=@n3.ViewportUV)
  n12 = BumpOffset(Coordinate=Add(A=ComponentMask(None=@n11) {g=true, r=true}, B=@n6), Height=Constant() {r=2})
  n13 = Fmod(A=Add(A=ComponentMask(None=@n12) {g=true}, B=@n9), B=Constant() {r=0.2})
  n14 = Add(A=Abs(None=ConstantBiasScale(None=@n10) {bias=-0.5, scale=2}), B=Abs(None=ConstantBiasScale(None=@n13) {bias=-0.5, scale=2}))
  n15 = Multiply(A=Time(), B=Constant() {r=0.5})
  n16 = Fmod(A=Add(A=ComponentMask(None=@n8) {r=true}, B=@n15), B=Constant() {r=1})
  n17 = BumpOffset(Coordinate=@n7, Height=Constant() {r=-4})
  n18 = Fmod(A=Add(A=ComponentMask(None=@n17) {g=true}, B=@n15), B=Constant() {r=1})
  n19 = Add(A=Abs(None=ConstantBiasScale(None=@n16) {bias=-0.5, scale=2}), B=Abs(None=ConstantBiasScale(None=@n18) {bias=-0.5, scale=2}))
  n20 = Multiply(A=@n1.TVEffectBrightness, B=Add(A=Multiply(A=@n14, B=@n19), B=@n19))
  n21 = SceneColor!(UVs=ComponentMask(None=@n3.ViewportUV) {g=true, r=true})
  n22 = Constant()
  n23 = Desaturation(None=FeatureLevelSwitch(Default=@n21.RGB, ES2=@n22, ES3_1=@n22, SM4_REMOVED=@n22, SM5=@n21.RGB, SM6=@n21.RGB), Fraction=Constant() {r=1})
  MaterialExpressionConstantClamp_5 = Saturate(None=OneMinus(None=ComponentMask(None=@n23) {r=true}))
  n24 = Power(Base=@MaterialExpressionConstantClamp_5, Exp=@n1.ColorPower)
  n25 = Constant() {r=1}
  n26 = Constant() {r=0.5}
  n27 = If(
   A=@n24
   B=Constant() {r=0.7}
   "A > B"=@n25
   "A == B"=@n25
   "A < B"=If(A=@n24, B=Constant() {r=0.4}, "A > B"=@n26, "A == B"=@n26, "A < B"=Constant())
  )
  MaterialExpressionVertexColor_1 = ParticleColor()
  MaterialFunctionCall(Emissive=Multiply(A=Multiply(A=@n20, B=@n27) {constA=1}, B=@MaterialExpressionVertexColor_1.RGB)) {materialFunction=&MF_FX_EmissiveAdjustments}
  Multiply(A=Power(Base=@n5, Exp=@n1.FallOff), B=Constant() {r=1})
  n28 = SphereMask(A=TextureCoordinate(), B=Constant2Vector() {g=0.5, r=0.5}, Hardness=ScalarParameter() {defaultValue=1, parameterName=Hardness}) {attenuationRadius=0.5, hardnessPercent=80}
  Multiply(
   A=Multiply(A=@MaterialExpressionVertexColor_1.A, B=Subtract(A=@n28) {constB=0.35})
   B=MaterialFunctionCall(Opacity In=@n28, DepthFade Distance=Constant() {r=500}) {materialFunction=&MF_VCVFX_DepthFade}.Result
  )
 ]
}
