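# Material for the Jigsaw-branded TV mesh, flagged for Nanite and Niagara mesh particles.
# A packed mask's R channel adds Particle Color tint, while G removes diffuse from that blend as a metal mask.
# Brightened diffuse, shaded by the desaturated normal map and a squared normal-versus-camera rim term, fakes metal reflection.
# That result, weighted by both masks, feeds MF_FX_EmissiveAdjustments as emissive.
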Material /DealershipSharedLibrary/Assets/ThirdPartyBranding/Jigsaw/Materials/M_Jigsaw_TV
graph {
 properties {
  assetImportData="<self>:InterchangeAssetImportData_0"
  bUsedWithNanite=true
  bUsedWithNiagaraMeshParticles=true
  usedShadingModels=Default Lit
 }
 nodes [
  n1 = TextureSampleParameter2D() {parameterName=Diffuse, texture=/DealershipSharedLibrary/Assets/ThirdPartyBranding/Jigsaw/Textures/T_Jigsaw_TV_D}
  n2 = TextureSampleParameter2D() {parameterName=Param, texture=/DealershipSharedLibrary/Assets/ThirdPartyBranding/Jigsaw/Textures/T_Jigsaw_TV_RGB}
  Add(A=Multiply(A=@n1.RGB, B=OneMinus(None=@n2.G)), B=Multiply(A=ParticleColor().RGB, B=@n2.R))
  n3 = Multiply(A=ScalarParameter() {defaultValue=3, parameterName=DiffuseBrightness}, B=@n1.RGB)
  n4 = TextureSampleParameter2D() {
   parameterName=Normal
   samplerType=SAMPLERTYPE_Normal
   texture=/DealershipSharedLibrary/Assets/ThirdPartyBranding/Jigsaw/Textures/T_Jigsaw_TV_N
  }
  n5 = CameraVectorWS()
  MaterialExpressionConstantClamp_2 = Saturate(None=Multiply(A=DotProduct(A=@n4.RGB, B=@n5), B=ScalarParameter() {defaultValue=2, parameterName=RimSharpness}))
  n6 = OneMinus(None=@MaterialExpressionConstantClamp_2)
  n7 = Multiply(A=Multiply(A=@n6, B=@n6), B=ScalarParameter() {defaultValue=2, parameterName=RimValue})
  n8 = Add(A=Multiply(A=@n3, B=Desaturation(None=@n4.RGB)), B=Multiply(A=@n3, B=Multiply(B=@n7)))
  n9 = Multiply(A=ScalarParameter() {defaultValue=3, parameterName=Metal reflection multiplier}, B=@n2.G)
  MaterialFunctionCall(Emissive=Add(A=Multiply(A=@n8, B=@n9), B=Multiply(A=@n8, B=@n2.R))) {materialFunction=&MF_FX_EmissiveAdjustments}
 ]
}
