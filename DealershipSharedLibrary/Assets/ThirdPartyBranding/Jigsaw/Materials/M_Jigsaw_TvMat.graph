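# Unlit Jigsaw-branded TV screen material, usable on Niagara mesh particles, showing a brightened, slightly desaturated scene texture under animated green-tinted static.
# Static blends two desaturated noise textures on tiled, pannable, flipbook-stepped UVs, with a sine-pulsed glitch intensity.
# The Billy texture, sampled normally and U-mirrored, masks where static gives way to the clean scene before the shared emissive-adjustment function.
# A separate 8x8 flipbook coordinate on UV1 takes speed and direction from dynamic parameters.
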Material /DealershipSharedLibrary/Assets/ThirdPartyBranding/Jigsaw/Materials/M_Jigsaw_TvMat
graph {
 properties {bUsedWithNanite=true, bUsedWithNiagaraMeshParticles=true, shadingModel=MSM_Unlit, usedShadingModels=Unlit}
 nodes [
  n1 = Constant3Vector() {constant=<rgba 0.1 1 0.2 0>}
  n2 = TextureCoordinate() {vTiling=0}
  n3 = Panner(
   Coordinate=Multiply(A=ComponentMask(None=@n2) {r=true}, B=ScalarParameter() {defaultValue=1, group=PatternOverlay, parameterName=PatternTileU})
   Time=ScalarParameter() {group=PatternOverlay, parameterName=Pattern_PanU}
  ) {speedX=0.5}
  n4 = TextureCoordinate() {uTiling=0}
  n5 = Panner(
   Coordinate=Multiply(A=ComponentMask(None=@n4) {g=true}, B=ScalarParameter() {defaultValue=1, group=PatternOverlay, parameterName=PatternTileV})
   Time=ScalarParameter() {group=PatternOverlay, parameterName=Pattern_PanV}
  ) {speedY=0.5}
  n6 = ScalarParameter() {defaultValue=2, group=FlipBook, parameterName=Columns}
  n7 = ScalarParameter() {defaultValue=2, group=FlipBook, parameterName=Rows}
  n8 = Multiply(A=AppendVector(A=ComponentMask(None=@n3) {r=true}, B=ComponentMask(None=@n5) {g=true}), B=AppendVector(A=Divide(B=@n6), B=Divide(B=@n7)))
  n9 = Multiply(A=Time(), B=ScalarParameter() {defaultValue=8, group=FlipBook, parameterName=FB_Speed})
  n10 = Floor(None=@n9)
  n11 = Divide(A=@n9, B=@n7)
  n12 = AppendVector(A=Divide(A=Fmod(A=@n10, B=@n7), B=@n7), B=Divide(A=Floor(None=@n11), B=@n6))
  n13 = Desaturation(None=TextureSample(UVs=Add(A=@n8, B=@n12)) {texture=/DealershipSharedLibrary/FX/Textures/Noises/T_Glitch_RGB}.RGB) {luminanceFactors=<rgba 0.3 0.59 0.11 0>}
  n14 = Multiply(A=@n8) {constB=2}
  n15 = Desaturation(None=TextureSample(UVs=Add(A=@n14, B=@n12)) {texture=/DealershipSharedLibrary/FX/Textures/Noises/T_NoiseGeneric_RGB}.RGB)
  n16 = Add(A=Add(A=Multiply(A=@n1, B=@n13), B=@n13), B=Add(A=Multiply(A=@n1, B=@n15), B=@n15))
  n17 = ScalarParameter() {defaultValue=0.15, parameterName=GlitchIntensity}
  n18 = Sine(None=Time()) {period=8}
  n19 = NamedRerouteUsage()
  n20 = Multiply(
   A=Desaturation(
    None=TextureSample(UVs=@n19) {texture=/DealershipSharedLibrary/Assets/ThirdPartyBranding/Jigsaw/Textures/T_Jigsaw_Scene_D}.RGB
    Fraction=Constant() {r=0.25}
   )
   B=Constant() {r=3}
  )
  n21 = Add(A=LinearInterpolate(A=@n16, B=Multiply(B=@n16) {constA=6}, Alpha=Multiply(A=@n17, B=@n18)), B=@n20)
  n22 = NamedRerouteUsage()
  n23 = ComponentMask(None=OneMinus(None=@n22)) {r=true}
  n24 = ComponentMask(None=@n22) {g=true}
  MaterialExpressionConstantClamp_11 = Saturate(None=TextureSample(UVs=AppendVector(A=@n23, B=@n24)) {texture=/DealershipSharedLibrary/Assets/ThirdPartyBranding/Jigsaw/Textures/T_Jigsaw_Billy_D}.RGB)
  MaterialExpressionConstantClamp_10 = Saturate(None=TextureSample(UVs=@n22) {texture=&T_Jigsaw_Billy_D}.RGB)
  MaterialFunctionCall(Emissive=LinearInterpolate(A=LinearInterpolate(A=@n21, B=@n20, Alpha=@MaterialExpressionConstantClamp_11), B=@n20, Alpha=@MaterialExpressionConstantClamp_10)) {materialFunction=&MF_FX_EmissiveAdjustments}
  n25 = TextureCoordinate() {coordinateIndex=1}
  n26 = AppendVector(A=Divide() {constB=8}, B=Divide() {constB=8})
  MaterialExpressionMeshEmitterDynamicParameter_0 = DynamicParameter() {paramNames [Speed,InvertSpeed,Param3,Param4]}
  n27 = Constant() {r=8}
  n28 = Constant() {r=-8}
  n29 = Ceil(None=@MaterialExpressionMeshEmitterDynamicParameter_0.InvertSpeed)
  n30 = Multiply(A=Time(), B=Multiply(A=@MaterialExpressionMeshEmitterDynamicParameter_0.Speed, B=LinearInterpolate(A=@n27, B=@n28, Alpha=@n29)))
  n31 = Divide(A=Fmod(A=Floor(None=@n30), B=Constant() {r=8})) {constB=8}
  n32 = Divide(A=Floor(None=Divide(A=@n30) {constB=8})) {constB=8}
  NamedRerouteDeclaration(None=Add(A=Multiply(A=@n25, B=@n26), B=AppendVector(A=@n31, B=@n32))) {name=Coordinate, nodeColor=<rgba 1 0 0.2812 1>}
 ]
}
